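; Parent Blueprint for AlleyTown electric pole props, assembled in the construction script from lower, middle and upper sections with a lamp.
; A location-derived random stream picks each section's yaw and the mid-section mesh, with the light pole turned 90 degrees from the middle.
; The spotlight snaps to the light pole's first socket plus a yaw offset, with an optional light cone.
; An exposed-asset-provider override registers mid-section mesh references.
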(event EventBeginPlay)

(event Collision|EventActorBeginOverlap
  (params (OtherActor :type Actor%20Object%20Reference)))

(event EventTick
  (params (DeltaSeconds :type Float%20%28single-precision%29)))

(fn BuildExposedAssetProviderVersion(Implementation)
  (params (Version :type Fort%20Exposed%20Asset%20Provider%20Version%20Structure%20%28by%20ref%29))
  (|CallParentFunction Version)
  (ExposedAssetProvider|AddVersion Version 1)
  (return))

(fn CollectAssetReferencestoAllow(Implementation)
  (params (Collector :type Fort%20Exposed%20Asset%20Reference%20Collector%20Structure%20%28by%20ref%29))
  (|CallParentFunction Collector)
  (for _array_element (Variables|Default|GetAR_MidSection)
    (ExposedAssetProvider|CollectIndirectAssetReference Collector (Utilities|ResolveSoftReference _array_element))
    (return)))

(fn SetLightPosition
  (bind _lightcone (Variables|Default|GetLightCone))
  (bind _spotlight (Variables|Default|GetSpotLight))
  (Transformation|SetWorldTransform _spotlight (Transformation|GetSocketTransform (Variables|Default|GetLightPole) (Utilities|Array|Get(acopy) (Transformation|GetAllSocketNames (Variables|Default|GetLightPole))) "RTS_World") false false)
  (Transformation|AddRelativeRotation _spotlight (Math|Rotator|MakeRotator 0.0 (Variables|Default|GetLightYaw)) false false)
  (Transformation|AddRelativeRotation (Variables|Default|GetLightCone) "-90.000000,0.000000,0.000000" false false)
  (if (Variables|Default|GetShowLightCone)
    (Rendering|SetVisibility _lightcone true false)
    (Transformation|SetRelativeScale3D _lightcone "2.000000,1.500000,4.000000")
    (Transformation|AddRelativeLocation _lightcone "0.000000,0.000000,-25.000000" false false)
    (else
      (Rendering|SetVisibility _lightcone false false))))

(fn SetMidSection
  (bind _randomstream (Variables|Default|GetRandomStream))
  (bind _rotationamount (Variables|Default|GetRotationAmount))
  (bind _outitem (Utilities|Array|RandomArrayItemfromStream _rotationamount _randomstream))
  (bind _returnvalue (Math|Rotator|MakeRotator 0.0 0.0 _outitem))
  (bind _midsection (Variables|Default|GetMidSection))
  (Components|StaticMesh|SetStaticMesh _midsection (Utilities|Array|RandomArrayItemfromStream (Variables|Default|GetAR_MidSection) (Variables|Default|GetRandomStream)))
  (Transformation|SetRelativeRotation _midsection _returnvalue false false)
  (Transformation|SetRelativeRotation (Variables|Default|GetLightPole) (Math|Rotator|MakeRotator (.roll _returnvalue) (.pitch _returnvalue) (- (.yaw _returnvalue) 9e+01)) false false))

(fn SetRandomStream
  (bind _returnvalue (Math|Random|MakeRandomStream (Math|Float|Truncate (* (Class|EnvironmentArtFuncLib|RandomFloatbyLocation (Transformation|GetActorLocation) (Variables|Default|GetTransitionDistance) true) 1e+04))))
  (Variables|Default|SetRandomStream _returnvalue))

(fn SetUpperSection
  (bind _uppersection (Variables|Default|GetUpperSection))
  (Transformation|SetRelativeRotation _uppersection (Math|Rotator|MakeRotator 0.0 0.0 (Utilities|Array|RandomArrayItemfromStream (Variables|Default|GetRotationAmount) (Variables|Default|GetRandomStream))) false false)
  (AddComponent|NODE_AddStaticMeshComponent-4 :RelativeTransform (Transformation|GetRelativeTransform _uppersection)))

(fn ConstructionScript
  (CallFunction|SetRandomStream)
  (Transformation|SetRelativeRotation (Variables|Default|GetLowerSection) (Math|Rotator|MakeRotator 0.0 0.0 (Utilities|Array|RandomArrayItemfromStream (Variables|Default|GetRotationAmount) (Variables|Default|GetRandomStream))) false false)
  (CallFunction|SetUpperSection)
  (CallFunction|SetMidSection)
  (CallFunction|SetLightPosition))
