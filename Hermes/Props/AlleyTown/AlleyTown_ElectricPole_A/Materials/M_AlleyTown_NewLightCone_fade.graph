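# Unlit additive material for a fake light cone on the AlleyTown electric pole, fading out as the camera nears it.
# A tinted vertical UV gradient is modulated by panning Perlin noise, while inverse Fresnel, depth fade and static noise soften the cone.
# Camera distance to the instance origin drives the near fade and switches the vertex offset between camera-aligning the mesh and collapsing it onto that origin.
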Material /Hermes/Props/AlleyTown/AlleyTown_ElectricPole_A/Materials/M_AlleyTown_NewLightCone_fade
graph {
 properties {bAutomaticallySetUsageInEditor=false, blendMode=BLEND_Additive, shadingModel=MSM_Unlit, usedShadingModels=Unlit}
 nodes [
  Constant3Vector()
  n1 = Distance(A=CameraPositionWS(), B=NamedRerouteUsage())
  n2 = ScalarParameter() {defaultValue=500, parameterName=NearFadeOffset}
  n3 = Saturate(None=Divide(A=Subtract(A=@n1, B=@n2), B=ScalarParameter() {defaultValue=2000, parameterName=NearFadeBlend}))
  n4 = Subtract(A=NamedRerouteUsage(), B=WorldPosition().XYZ)
  If(
   A=@n3
   "A > B"=MaterialFunctionCall!(
    WorldPosition
    PivotLocation
    WS Normals
    Custom Object Basis 1=Constant3Vector() {constant=<rgba 1 0 0 0>}
    Custom Object Basis 2=Constant3Vector() {constant=<rgba 0 0 1 0>}
    Custom Object Basis 3=Constant3Vector() {constant=<rgba 0 1 0 0>}
   ) {materialFunction=/Engine/Functions/Engine_MaterialFunctions02/WorldPositionOffset/AlignMeshToTheCamera}.World Position Offset
   "A == B"=@n4
   "A < B"=@n4
  )
  Multiply()
  n5 = MaterialFunctionCall() {materialFunction=&LinearGradient}
  n6 = OneMinus(None=@n5.VGradient)
  n7 = Multiply(A=@n5.VGradient, B=Saturate(None=Power(Base=@n6)))
  n8 = ScalarParameter() {defaultValue=0.5, parameterName=Noise Contrast}
  n9 = Panner(Coordinate=TextureCoordinate(), Speed=AppendVector(A=Constant(), B=ScalarParameter() {defaultValue=0.025, parameterName=PanningSpeed}))
  Multiply(
   A=Multiply(A=Multiply(A=@n7) {constB=1.5}, B=VectorParameter() {parameterName=Cone Color, defaultValue=<rgba 0.604 1 0.8046 1>}.RGB)
   B=LinearInterpolate(
    A=Subtract(B=@n8) {constA=0.5}
    B=Add(A=@n8)
    Alpha=TextureSample(UVs=@n9) {samplerType=SAMPLERTYPE_Masks, texture=/Game/Global/Textures/Default/Noise/T_Generic_Perlin}.G
   ) {constA=0.2}
  )
  n10 = Fresnel() {baseReflectFraction=0, exponent=0.25}
  n11 = Multiply(
   A=DepthFade(
    Opacity=ScalarParameter() {defaultValue=1, parameterName=Opacity}
    FadeDistance=ScalarParameter() {defaultValue=50, parameterName=DepthFadeDistance}
   )
   B=LinearInterpolate(Alpha=TextureSample() {samplerType=SAMPLERTYPE_Masks, texture=&T_Generic_Perlin}.B) {constA=0.7}
  )
  Multiply(A=Multiply(A=OneMinus(None=@n10), B=@n11), B=Multiply(A=@n11, B=@n3))
  NamedRerouteDeclaration(None=TransformPosition(None=Constant3Vector()) {transformSourceType=TRANSFORMPOSSOURCE_Instance, transformType=TRANSFORMPOSSOURCE_World}) {name=OriginPoint, nodeColor=<rgba 1 0.4453 0 1>}
 ]
}
